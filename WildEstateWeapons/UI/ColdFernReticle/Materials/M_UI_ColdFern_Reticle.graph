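# Translucent UI material that draws the ColdFern weapon reticle: brackets, four-sided crosshair and center dot, each with an outline.
# Spread parameters pass through shared remap and reticle-UV functions to offset the crosshair lines and build rotated four-sided custom UVs.
# Brackets threshold an IronLog distance-field texture; procedural shapes supply the dot, lines and sine-masked rounded boxes.
# Layers combine into color and opacity, with per-part opacity controls under Master Opacity.
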Material /WildEstateWeapons/UI/ColdFernReticle/Materials/M_UI_ColdFern_Reticle
graph {
 properties {bAutomaticallySetUsageInEditor=false, blendMode=BLEND_Translucent, materialDomain=MD_UI, usedShadingModels=Default Lit}
 nodes [
  NamedRerouteDeclaration(None=MaterialFunctionCall(Size=Constant() {r=0.2}, OutlineThickness=Constant() {r=0.01}) {materialFunction=&MF_UI_Circle}.Outline) {name=Circle, nodeColor=<rgba 0 0 0 1>}
  n1 = NamedRerouteUsage()
  n2 = Constant() {r=0.7}
  n3 = Constant() {r=0.2}
  n4 = Constant() {r=0.05}
  NamedRerouteDeclaration(None=Multiply(
    A=MaterialFunctionCall(UVs=@n1, Size=@n2, CornerRoundness=@n3, OutlineThickness=@n4) {materialFunction=&MF_UI_Box}.Outline
    B=NamedRerouteUsage()
   )) {name=Outline, nodeColor=<rgba 0 0 0 1>}
  n5 = TextureSample(UVs=NamedRerouteUsage()) {samplerSource=SSM_Clamp_WorldGroupSettings, texture=/WildEstateWeapons/UI/IronLogReticle/Textures/T_UI_IronLog_Reticle_Bracket}
  NamedRerouteDeclaration(None=MaterialFunctionCall(DistanceField=@n5.RGB, Threshold=ScalarParameter() {defaultValue=0.25, parameterName=Bracket Outline thickness}) {materialFunction=&MF_UI_AntiAliasing}.Outer) {name=Bracket Outline, nodeColor=<rgba 0 0 0 1>}
  n6 = Add(A=NamedRerouteUsage(), B=NamedRerouteUsage())
  n7 = NamedRerouteUsage()
  NamedRerouteDeclaration(None=LinearInterpolate(Alpha=Add(A=@n6, B=@n7))) {name=Final Color, nodeColor=<rgba 0.6875 0 1 1>}
  n8 = Multiply(A=NamedRerouteUsage(), B=ScalarParameter() {defaultValue=0.5, parameterName=Bracket Outline opacity})
  n9 = NamedRerouteUsage()
  n10 = Multiply(A=@n9, B=ScalarParameter() {defaultValue=0.6, parameterName=Brackets Opacity}) {constB=0.6}
  n11 = Add(
   A=Add(A=LinearInterpolate(A=@n8, B=@n10, Alpha=@n9), B=NamedRerouteUsage())
   B=Multiply(A=NamedRerouteUsage(), B=ScalarParameter() {defaultValue=0.5, parameterName=Cross Hair outline opacity}) {constB=0.6}
  )
  n12 = NamedRerouteUsage()
  n13 = LinearInterpolate(A=Multiply(A=ScalarParameter() {defaultValue=0.5, parameterName=Center Dot Outline Opacity}, B=NamedRerouteUsage()), B=@n12, Alpha=@n12)
  NamedRerouteDeclaration(None=Multiply(A=Add(A=@n11, B=@n13), B=ScalarParameter() {defaultValue=1, parameterName=Master Opacity})) {name=Opacity, nodeColor=<rgba 0 1 0.3672 1>}
  n14 = Constant() {r=4}
  n15 = ScalarParameter() {defaultValue=0.5, parameterName=Bracket Scale}
  n16 = MaterialFunctionCall!(
   Positional Spread Offset=ScalarParameter() {defaultValue=60, parameterName=Shotgun_Spread}
   Accuracy Spread Offset
   Min Value
   "Input Range [Min,Max]"
   "Remapped Range [Min,Max]"
  ) {materialFunction=/FortUILibrary/Systems/Reticles/MaterialFunctions/MF_UI_Reticle_Remap_Spread}
  NamedRerouteDeclaration(None=MaterialFunctionCall(
    UVs=MaterialFunctionCall!(
     Reticle UVs
     Reticle Sides=@n14
     Reticle Scale
     Reticle Rotation
     Prong Scale=@n15
     Prong Rotation
     Spread Offset=@n16.Result
     Positional Offset
     "Spread Input Range [Min,Max]"
     "Spread Remapped Range [Min,Max]"
    ) {materialFunction=/FortUILibrary/Systems/Reticles/MaterialFunctions/MF_UI_Reticle_UVs}.UVs
    "Rotation Angle (0-1)"=Constant() {r=0.125}
   ) {materialFunction=&CustomRotator}.Rotated Values) {name=CustomUV, nodeColor=<rgba 0 0 0 1>}
  n17 = MaterialFunctionCall(
   Size=ScalarParameter() {defaultValue=0.025, parameterName=Center Dot Scale}
   OutlineThickness=ScalarParameter() {defaultValue=0.015, parameterName=Center Dot Outline Thickness}
  ) {materialFunction=&MF_UI_Circle}
  NamedRerouteDeclaration(None=@n17.Fill) {name=Center Dot, nodeColor=<rgba 0.7578 1 0 1>}
  NamedRerouteDeclaration(None=Power(Base=@n17.Outline) {constExponent=0.01}) {name=Center dot outline, nodeColor=<rgba 0.9219 0 1 1>}
  n18 = NamedRerouteUsage()
  n19 = Constant() {r=0.7}
  n20 = Constant() {r=0.2}
  n21 = Constant() {r=0.02}
  NamedRerouteDeclaration(None=Multiply(
    A=MaterialFunctionCall(UVs=@n18, Size=@n19, CornerRoundness=@n20, OutlineThickness=@n21) {materialFunction=&MF_UI_Box}.Outline
    B=NamedRerouteUsage()
   )) {name=Reticle, nodeColor=<rgba 0 0 0 1>}
  n22 = Add(A=TextureCoordinate()) {constB=0.25}
  n23 = Sine(None=ComponentMask(None=@n22) {r=true})
  n24 = Sine(None=ComponentMask(None=@n22) {g=true})
  NamedRerouteDeclaration(None=Min(A=Step(X=@n23) {constX=0, constY=-0.2}, B=Step(X=@n24) {constX=0, constY=-0.2})) {name=Mask, nodeColor=<rgba 0 0 0 1>}
  NamedRerouteDeclaration(None=Min(A=Step(X=@n23) {constX=0, constY=-0.25}, B=Step(X=@n24) {constX=0, constY=-0.25})) {name=Mask outline, nodeColor=<rgba 0 0 0 1>}
  n25 = Constant()
  n26 = Constant() {r=0.01}
  NamedRerouteDeclaration(None=Add(
    A=TextureCoordinate()
    B=MaterialFunctionCall!(UV, SpeedX=@n25, SpeedY=@n25, Intensity=@n26) {materialFunction=/WildEstateWeapons/UI/MaterialFunctions/MF_UI_WildEstate_Distort}.Result
   )) {name=Distort, nodeColor=<rgba 0 0 0 1>}
  n27 = ScalarParameter() {defaultValue=100, parameterName=Spread}
  n28 = MaterialFunctionCall!(
   UVs=MaterialFunctionCall(
    Reticle Sides=Constant() {r=4}
    Reticle Scale=Constant() {r=0.15}
    Reticle Rotation=Constant() {r=0.125}
    Spread Offset=MaterialFunctionCall(Positional Spread Offset=@n27) {materialFunction=&MF_UI_Reticle_Remap_Spread}.Result
   ) {materialFunction=&MF_UI_Reticle_UVs}.UVs
   PointA=Constant2Vector() {g=-1}
   PointB=Constant2Vector() {g=1}
   Thickness=ScalarParameter() {defaultValue=0.2, parameterName=Crosshair thickness}
   OutlineThickness=ScalarParameter() {defaultValue=0.3, parameterName=Crosshair outline thickness}
   OutlineAlignment
   GlowMin
   GlowMax
  ) {materialFunction=/FortUILibrary/MaterialFunctions/Shapes/MF_UI_LineSegment}
  NamedRerouteDeclaration(None=@n28.Fill) {name=MiddlePart, nodeColor=<rgba 0 0 0 1>}
  NamedRerouteDeclaration(None=@n28.Outline) {name=MiddleLinesOutline, nodeColor=<rgba 0 0 0 1>}
  NamedRerouteDeclaration(None=MaterialFunctionCall(DistanceField=@n5.RGB, Threshold=ScalarParameter() {defaultValue=0.4, parameterName=Bracket thickness}) {materialFunction=&MF_UI_AntiAliasing}.Outer) {name=Brackets, nodeColor=<rgba 0 0 0 1>}
  NamedRerouteUsage()
  NamedRerouteUsage()
 ]
}
