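# Translucent UI material for the ColdFern reticle's ammo counter: a segmented radial progress bar over a textured backplate.
# AnimateProgress drives the arc fill, snapped to steps normally and continuous as IsReload goes to 1.
# Rotated UI UVs feed radial arc and SDF circle masks, with ping-pong separators cutting segments and anti-aliasing smoothing edges.
# MainColor and EmptyColor blend for color; backplate and master opacity parameters shape transparency.
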Material /WildEstateWeapons/UI/ColdFernReticle/Materials/M_UI_ColdFerne_AmmoCounter
graph {
 properties {bAutomaticallySetUsageInEditor=false, blendMode=BLEND_Translucent, materialDomain=MD_UI, usedShadingModels=Default Lit}
 nodes [
  Constant() {r=1}
  n1 = Multiply(A=NamedRerouteUsage(), B=LinearInterpolate(A=NamedRerouteUsage(), Alpha=ScalarParameter() {parameterName=IsReload}) {constA=1})
  LinearInterpolate(
   A=VectorParameter() {parameterName=MainColor, defaultValue=<rgba 0 0 0 1>}.RGB
   B=VectorParameter() {parameterName=EmptyColor, defaultValue=<rgba 1 1 1 1>}.RGB
   Alpha=@n1
  )
  n2 = Constant()
  MaterialFunctionCall!(UV, SpeedX=@n2, SpeedY=@n2, Intensity=Constant()) {materialFunction=/WildEstateWeapons/UI/MaterialFunctions/MF_UI_WildEstate_Distort}
  n3 = NamedRerouteUsage()
  n4 = ScalarParameter() {defaultValue=0.4, parameterName=Backplate opacity}
  Multiply(A=LinearInterpolate(A=@n1, B=@n3, Alpha=Multiply(A=@n3, B=@n4)), B=ScalarParameter() {defaultValue=0.8, parameterName=Master opacity})
  n5 = MaterialFunctionCall(UVs=Add(A=TextureCoordinate()) {constB=0.0075}, Texture Scale=Constant() {r=0.56}) {materialFunction=&ScaleUVsByCenter}
  NamedRerouteDeclaration(None=MaterialFunctionCall(
    DistanceField=TextureSample(UVs=@n5.UVs) {samplerSource=SSM_Clamp_WorldGroupSettings, texture=/WildEstateWeapons/UI/ColdFernReticle/Textures/T_UI_ColdFern_AmmoBackplate_128}.RGB
    Threshold=Constant() {r=0.58}
   ) {materialFunction=&MF_UI_AntiAliasing}.Outer) {name=Mask, nodeColor=<rgba 0 0 0 1>}
  n6 = ScalarParameter() {desc="Angle from 0 point and move clockwise.\r\n0 - 1", group=Progress_Layout, parameterName=StartAngle, sliderMax=1, sortPriority=5}
  n7 = ScalarParameter() {
   defaultValue=0.72
   desc="Angle from 0 point and move anticlockwise.\r\n0 - 1"
   group=Progress_Layout
   parameterName=EndAngle
   sliderMax=1
   sortPriority=6
  }
  n8 = LinearInterpolate(
   A=ScalarParameter() {group=Progress_Layout, parameterName=ProgressStart}
   B=ScalarParameter() {defaultValue=1, group=Progress_Layout, parameterName=ProgressEnd}
   Alpha=ScalarParameter() {defaultValue=1, group=Progress_Layout, parameterName=AnimateProgress}
  )
  n9 = LinearInterpolate(
   A=MaterialFunctionCall!("Value (0-1)"=@n8, Steps=NamedRerouteUsage()) {materialFunction=/FortUILibrary/MaterialFunctions/Basics/MF_UI_StepRatio}.Stepped Ratio
   B=@n8
   Alpha=ScalarParameter() {parameterName=IsReload}
  )
  n10 = MaterialFunctionCall!(UV=NamedRerouteUsage(), "Arc Start Angle (0-1)"=@n6, "Arc End Angle (0-1)"=@n7, "Fill Ratio (0-1)"=@n9, Fill Clockwise=Constant() {r=1}) {materialFunction=/FortUILibrary/MaterialFunctions/Basics/MF_UI_RadialArc}
  NamedRerouteDeclaration(None=@n10.Arc Mask) {name=ArcGradient, nodeColor=<rgba 0 0 0 1>}
  n11 = MaterialFunctionCall!(
   Value PX=ScalarParameter() {defaultValue=10, desc=The size of gap between segements in px., group=Segments, parameterName=SeparatorThickness_PX}
   UV Pixel Size
   X or Y
  ) {materialFunction=/FortUILibrary/MaterialFunctions/Basics/MF_UI_PixelsToNormalized}
  n12 = Multiply(A=@n11."Normalized (>1px)", B=Multiply(A=NamedRerouteUsage()) {constB=0.25})
  n13 = ScalarParameter() {defaultValue=0.05, group=Progress_Layout, parameterName=SeparatorSharpness, sortPriority=42}
  n14 = ScalarParameter() {defaultValue=36, desc=Number of segments in progress bar, group=Segments, parameterName=SegmentsCount, sliderMin=1, sortPriority=2}
  n15 = Multiply(A=NamedRerouteUsage(), B=NamedRerouteDeclaration(None=Floor(None=@n14)) {name=SegmentsCount, nodeColor=<rgba 0 0 0 1>})
  n16 = Constant() {r=1}
  n17 = ConstantBiasScale(None=NamedRerouteUsage()) {bias=-0.5, scale=2}
  n18 = Saturate(None=Multiply(
    A=MaterialFunctionCall!(Value=@n15, Scale=@n16, Height) {materialFunction=/FortUILibrary/MaterialFunctions/Basics/MF_UI_PingPong}.Result
    B=Length(None=@n17) {is=class MaterialExpressionLength}
   ))
  NamedRerouteDeclaration(None=SmoothStep(Min=Subtract(A=@n12, B=@n13), Max=Add(A=@n12, B=@n13), Value=Subtract(A=@n18, B=@n11."Normalized (>1px)")) {constMax=0.01}) {name=SegmentsMask, nodeColor=<rgba 0 0 0 1>}
  NamedRerouteDeclaration(None=Ceil(None=@n10.Radial Fill)) {name=ArcProgressMask, nodeColor=<rgba 0 0 0 1>}
  n19 = NamedRerouteUsage()
  n20 = MaterialFunctionCall(
   UVs=MaterialFunctionCall(UVs=@n19, Texture Scale=Constant() {r=1.185}) {materialFunction=&ScaleUVsByCenter}.UVs
   Size=ScalarParameter() {defaultValue=0.6, parameterName=Progress Bar Size}
   OutlineThickness=ScalarParameter() {defaultValue=0.02, parameterName=Progress Bar Thickness}
  ) {materialFunction=&MF_UI_Circle}
  n21 = MaterialFunctionCall(
   UV=MaterialFunctionCall(UVs=@n19, "Rotation Angle (0-1)"=Constant() {r=0.5}) {materialFunction=&CustomRotator}.Rotated Values
   "Arc Start Angle (0-1)"=ScalarParameter() {defaultValue=0.999, parameterName=Start Angle}
   "Arc End Angle (0-1)"=ScalarParameter() {defaultValue=0.25, parameterName=End Angle, sliderMax=1}
   "Fill Ratio (0-1)"=NamedRerouteUsage()
  ) {materialFunction=&MF_UI_RadialArc}
  NamedRerouteDeclaration(None=Multiply(
    A=NamedRerouteDeclaration(None=@n20.Outline) {name=Circle Mask, nodeColor=<rgba 0 0 0 1>}
    B=SmoothStep(Value=@n21.Radial Fill) {constMax=0.01}
   )) {name=Alternative Mask, nodeColor=<rgba 1 0.1406 0 1>}
  NamedRerouteDeclaration(None=Subtract(
    A=MaterialFunctionCall() {materialFunction=&GetUserInterfaceUV}.Normalized UV
    B=ScalarParameter() {defaultValue=-0.192, parameterName=UVCorrection}
   )) {name=UVs, nodeColor=<rgba 0 0 0 1>}
  n22 = ScalarParameter() {
   defaultValue=0.75
   desc="Rotate the start point of the progress bar.\r\n0 - top\r\n0.25 - left\r\n0.5 - bottom\r\n0.75 - right"
   group=Progress_Layout
   parameterName=RotationAngle
  }
  NamedRerouteDeclaration(None=MaterialFunctionCall(UVs=NamedRerouteUsage(), "Rotation Angle (0-1)"=@n22) {materialFunction=&CustomRotator}.Rotated Values) {name=RotatorUVs, nodeColor=<rgba 0 0 0 1>}
  n23 = ScalarParameter() {
   defaultValue=0.72
   desc="Angle from 0 point and move anticlockwise.\r\n0 - 1"
   group=Progress_Layout
   parameterName=EndAngle
   sliderMax=1
   sortPriority=6
  }
  n24 = OneMinus(None=@n23)
  n25 = ScalarParameter() {desc="Angle from 0 point and move clockwise.\r\n0 - 1", group=Progress_Layout, parameterName=StartAngle, sliderMax=1, sortPriority=5}
  n26 = ScalarParameter() {desc="Will fill the progress bar clockwise.\r\nbool", group=Progress_Layout, parameterName=FillClockwise, sliderMax=1, sortPriority=40}
  n27 = LinearInterpolate(A=LinearInterpolate(A=@n24, B=@n25, Alpha=@n26), B=LinearInterpolate(A=@n25, B=@n24, Alpha=@n26), Alpha=NamedRerouteUsage())
  n28 = NamedRerouteUsage()
  n29 = NamedRerouteUsage()
  n30 = ScalarParameter() {defaultValue=1, parameterName=CircleScale}
  n31 = NamedRerouteDeclaration(None=MaterialFunctionCall(
    UVs=MaterialFunctionCall(UVs=@n29, Texture Scale=@n30) {materialFunction=&ScaleUVsByCenter}.UVs
    Size=ScalarParameter() {defaultValue=0.712, parameterName=CircleSize}
    OutlineThickness=ScalarParameter() {defaultValue=0.028, parameterName=CircleThickness}
   ) {materialFunction=&MF_UI_Circle}.SDFOutline) {name=SDF_CircleBar, nodeColor=<rgba 0 0 0 1>}
  NamedRerouteDeclaration(None=MaterialFunctionCall(DistanceField=LinearInterpolate(A=@n27, B=@n31, Alpha=@n28)) {materialFunction=&MF_UI_AntiAliasing}.Inner) {name=ProgressMask, nodeColor=<rgba 0 0 0 1>}
  NamedRerouteDeclaration(None=OneMinus(None=@n9)) {name=Progress, nodeColor=<rgba 0 0 0 1>}
  NamedRerouteUsage()
 ]
}
