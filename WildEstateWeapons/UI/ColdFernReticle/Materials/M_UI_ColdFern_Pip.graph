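# Translucent UI material that draws a pip for the ColdFern weapon reticle, with an outline behind its main fill.
# Box and circle material functions build rounded-box and dot fills, each with a larger backplate, shared through named reroutes.
# The pip mask blends outline color into main color, while opacity blends the dimmer backplate with the pip.
# A UV chain adds a slight vertical offset and a distortion call whose speed and intensity are zero.
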Material /WildEstateWeapons/UI/ColdFernReticle/Materials/M_UI_ColdFern_Pip
graph {
 properties {bAutomaticallySetUsageInEditor=false, blendMode=BLEND_Translucent, materialDomain=MD_UI, usedShadingModels=Default Lit}
 nodes [
  n1 = TextureCoordinate()
  n2 = ComponentMask(None=@n1) {g=true}
  n3 = Constant() {r=-0.01}
  n4 = Constant()
  Add(
   A=AppendVector(A=ComponentMask(None=@n1) {r=true}, B=Add(A=@n2, B=@n3))
   B=MaterialFunctionCall!(UV, SpeedX=@n4, SpeedY=@n4, Intensity=Constant()) {materialFunction=/WildEstateWeapons/UI/MaterialFunctions/MF_UI_WildEstate_Distort}.Result
  )
  Constant2Vector() {g=0.75, r=0.25}
  Constant2Vector() {g=0.98, r=0.48}
  n5 = Multiply(A=NamedRerouteUsage(), B=ScalarParameter() {defaultValue=0.9, parameterName=PipOpacity})
  LinearInterpolate(A=Multiply(A=ScalarParameter() {defaultValue=0.4, parameterName=Backplate opacity}, B=NamedRerouteUsage()), B=@n5, Alpha=@n5)
  LinearInterpolate(
   A=VectorParameter() {parameterName=OutlineColor, defaultValue=<rgba 0 0 0 1>}.RGB
   B=VectorParameter() {parameterName=MainColor, defaultValue=<rgba 1 1 1 1>}.RGB
   Alpha=Multiply(A=NamedRerouteUsage(), B=ScalarParameter() {defaultValue=0.9, parameterName=PipOpacity})
  )
  NamedRerouteDeclaration(None=MaterialFunctionCall(Size=Constant2Vector() {g=0.6, r=0.3}, CornerRoundness=Constant() {r=0.3}) {materialFunction=&MF_UI_Box}.Fill) {name=Box, nodeColor=<rgba 0 0 0 1>}
  NamedRerouteDeclaration(None=MaterialFunctionCall(Size=Constant2Vector() {g=0.8, r=0.5}, CornerRoundness=Constant() {r=0.5}) {materialFunction=&MF_UI_Box}.Fill) {name=Backplate, nodeColor=<rgba 0 0.2734 1 1>}
  NamedRerouteDeclaration(None=MaterialFunctionCall(Size=Constant() {r=0.5}) {materialFunction=&MF_UI_Circle}.Fill) {name=Dot, nodeColor=<rgba 0.7109 0 1 1>}
  NamedRerouteDeclaration(None=MaterialFunctionCall(Size=Constant() {r=0.8}) {materialFunction=&MF_UI_Circle}.Fill) {name=Dot backplate, nodeColor=<rgba 0 1 0.3906 1>}
  NamedRerouteUsage()
  NamedRerouteUsage()
  NamedRerouteUsage()
 ]
}
