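# Reticle extension widget for the ColdFern dual-wield weapon, showing spread and ammo for an SMG and an offhand shotgun.
# On view target changes it rebinds delegates on the equipped SMG and the offhand component's shotgun.
# Tick feeds both spreads into reticle material parameters, while ammo events update a segmented SMG counter and two shotgun pips.
# Reload animation speed scales to reload time; a timer fades the ammo display, interrupted by firing or reloading.
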WidgetBlueprint /WildEstateWeapons/UI/ColdFernReticle/WBP_ColdFern_DualWield_Reticle
animations [
 Anim_Reload {
  binding_count=4
  length_seconds=1
  object=&self:Anim_Reload
  sequence {bindings [
    Image_Reticle {
     id=#1
     possessed_class=Image
     tracks [{
       type=MovieSceneWidgetMaterialTrack
       sections [{
         type=MovieSceneParameterSection
         end_frame=20
         start_frame=0
         channels [Master Opacity {type=MovieSceneScriptingFloatChannel, keys [[0,0.25],[19,0.25],[20,1]]}]
        }]
      }]
    }
    Image_AmmoPip2 {
     id=#2
     possessed_class=Image
     tracks [{
       type=MovieSceneWidgetMaterialTrack
       sections [{
         type=MovieSceneParameterSection
         end_frame=20
         start_frame=0
         channels [Backplate opacity {type=MovieSceneScriptingFloatChannel, keys [[0,0],[19,0],[20,0.5]]}]
        }]
      }]
    }
    Image_AmmoPip1 {
     id=#3
     possessed_class=Image
     tracks [{
       type=MovieSceneWidgetMaterialTrack
       sections [{
         type=MovieSceneParameterSection
         end_frame=20
         start_frame=0
         channels [Backplate opacity {type=MovieSceneScriptingFloatChannel, keys [[0,0],[19,0],[20,0.5]]}]
        }]
      }]
    }
    Image_AmmoCounter {
     id=#4
     possessed_class=Image
     tracks [{
       type=MovieSceneWidgetMaterialTrack
       sections [{
         type=MovieSceneParameterSection
         end_frame=20
         start_frame=0
         channels [
          IsReload {type=MovieSceneScriptingFloatChannel, const [1,2]}
          AnimateProgress {type=MovieSceneScriptingFloatChannel, keys [[0,0],[20,1]]}
         ]
        }]
      }]
    }
   ]}
 }
 Anim_HideAmmo {
  binding_count=1
  length_seconds=0.5
  object=&self:Anim_HideAmmo
  sequence {bindings [AmmoCounter {
     id=#5
     possessed_class=Overlay
     tracks [{
       type=MovieSceneFloatTrack
       display_name=Render Opacity
       sections [{type=MovieSceneFloatSection, end_frame=10, start_frame=0, channels [None {type=MovieSceneScriptingFloatChannel, keys [[0,1],[10,0]]}]}]
      }]
    }]}
 }
]
graphs [
 BindSMGEvents {nodes [
   BindSMGEvents() {is=title K2Node_FunctionEntry}
   > n1 = Cast To B_ColdFern_DualWeapon!(
    execute
    Object=GetEquippedWeapon!(self:Fort Weapon Reticle Extension Widget Base Object Reference) -> (ReturnValue:Fort Weapon Object Reference)
   ) -> (then, CastFailed, AsB Cold Fern Dual Weapon:B Cold Fern Dual Weapon Object Reference) {is=title K2Node_DynamicCast}
   > n2 = Set SMG_WeaponRef(SMG_WeaponRef:B Cold Fern Dual Weapon Object Reference=@n1) -> (Output_Get:B Cold Fern Dual Weapon Object Reference)
   > Bind Event to On Local Ammo Changed(self:Fort Weapon Object Reference=@n2, Delegate=Create Event()) {is=title K2Node_AddDelegate}
   > Bind Event to On Local Reload Started(self:Fort Weapon Object Reference=@n2, Delegate=Create Event()) {is=title K2Node_AddDelegate}
   > Bind Event to On Local Reload Cancelled(self:Fort Weapon Object Reference=@n2, Delegate=Create Event()) {is=title K2Node_AddDelegate}
  ]}
 BindShotgunEvents {nodes [
   n1 = BindShotgunEvents!() -> (then, ShotgunWeapon:Fort Weapon Object Reference) {is=title K2Node_FunctionEntry}
   > n2 = Cast To B_DualWeapon_Shotgun_WildEstate_ColdFern!(execute, Object=@n1) -> (then, CastFailed, AsB Dual Weapon Shotgun Wild Estate Cold Fern:B Dual Weapon Shotgun Wild Estate Cold Fern Object Reference) {is=title K2Node_DynamicCast}
   > n3 = Set Shotgun_WeaponBPRef(Shotgun_WeaponBPRef:B Dual Weapon Shotgun Wild Estate Cold Fern Object Reference=@n2) -> (Output_Get:B Dual Weapon Shotgun Wild Estate Cold Fern Object Reference)
   > Bind Event to On Local Ammo Changed(self=@n3, Delegate=Create Event())
   > Bind Event to On Start Firing(self:B Dual Weapon Cold Fern Generic Athena Object Reference=@n3, Delegate=Create Event()) {is=title K2Node_AddDelegate}
   > Bind Event to On Persistent Fire Stopped(self:B Dual Weapon Cold Fern Generic Athena Object Reference=@n3, Delegate=Create Event()) {is=title K2Node_AddDelegate}
   PrintString!(
    execute=@n2.CastFailed
    InString="BindShotgunEvents: Could not cast shotgun to B_DualWeapon_Shotgun_WildEstate_ColdFern!"
    bPrintToScreen
    bPrintToLog
    TextColor=(R=1,G=0.0701,B=0,A=1)
    Duration=5
    Key
   ) -> (then)
  ]}
 EventGraph {nodes [
   n1 = OnViewTargetChanged!() -> (
    OutputDelegate
    then
    PlayerController:Fort Player Controller Object Reference
    OldViewTarget:Actor Object Reference
    NewViewTarget:Actor Object Reference
   ) {is=title K2Node_CustomEvent}
   > n2 = Sequence()
   > UnbindSMGEvents()
   > BindSMGEvents() {is=title K2Node_CallFunction}
   n3 = Set OffhandWeaponComp(
    execute=@n2.then_1
    OffhandWeaponComp:Fort Offhand Component Object Reference=GetComponentByClass!(self:Actor Object Reference=@n1.NewViewTarget, ComponentClass:Actor Component Class Reference=/Script/FortniteGame.FortOffhandComponent) -> (ReturnValue:Fort Offhand Component Object Reference)
   ) -> (Output_Get:Fort Offhand Component Object Reference)
   > n4 = Is Valid(InputObject=@n3)
   PrintString(
    execute=@n4.Is Not Valid
    InString="OnViewTargetChanged: Culd not find OffhandWeaponComponent!"
    TextColor=(R=1,G=0.0701,B=0,A=1)
    Duration=5
   )
   n5 = GetCurrentEquippedOffhandWeapon!(
    execute=@n4.Is Valid
    self:Fort Offhand Component Object Reference=Get OffhandWeaponComp() -> (OffhandWeaponComp:Fort Offhand Component Object Reference)
   ) -> (then, ReturnValue:Fort Weapon Object Reference)
   > UnbindShotgunEvents()
   > BindShotgunEvents!(execute, self, ShotgunWeapon:Fort Weapon Object Reference=@n5) -> (then) {is=title K2Node_CallFunction}
   n6 = OnLocalReloadStarted!() -> (OutputDelegate, then, ReloadTime:float=~, ReloadType:EFortWeaponReloadType Enum=ReloadWholeClip) {is=title K2Node_CustomEvent}
   > PlayAnimation(
    InAnimation=Get Anim_Reload() -> (Anim_Reload:Widget Animation Object Reference)
    PlaybackSpeed=SafeDivide(A=1, B=@n6.ReloadTime)
    bRestoreState=true
   )
   > BindToAnimationFinished(Animation=Get Anim_Reload(), Delegate=Create Event())
   > Clear and Invalidate Timer by Handle(Handle=Get AutoHideTime() -> (AutoHideTime:Timer Handle Structure))
   > InterruptFadeOut()
   Firing Started() {is=title K2Node_CustomEvent}
   > Clear and Invalidate Timer by Handle(Handle=Get AutoHideTime())
   > InterruptFadeOut()
   n7 = HideAmmoBarsTimer() {is=title K2Node_CustomEvent}
   n8 = HideBars() {is=title K2Node_CustomEvent}
   n9 = Set Timer by Event(execute=@n7.then, Delegate=@n8, Time=2, bLooping=false)
   > Set AutoHideTime(AutoHideTime:Timer Handle Structure=@n9) -> (Output_Get:Timer Handle Structure)
   Firing Stopped() {is=title K2Node_CustomEvent}
   > HideAmmoBarsTimer() {is=title K2Node_CallFunction}
   Event Destruct()
   n10 = Get OwningPlayerControllerRef() -> (OwningPlayerControllerRef:Fort Player Controller Object Reference)
   > n11 = Is Valid(InputObject=@n10)
   PrintString(execute=@n11.Is Not Valid, InString="EventDestruct: OwningPlayerController is invalid!", TextColor=(R=1,G=0.0701,B=0,A=1), Duration=5)
   Unbind Event from On View Target Changed(execute=@n11.Is Valid, self:Fort Player Controller Object Reference=@n10, Delegate=Create Event()) {is=title K2Node_RemoveDelegate}
   > UnbindSMGEvents()
   > UnbindShotgunEvents()
   Event Construct()
   > n12 = Cast To FortPlayerController(Object=GetOwningPlayer())
   > n13 = Set OwningPlayerControllerRef(OwningPlayerControllerRef:Fort Player Controller Object Reference=@n12) -> (Output_Get:Fort Player Controller Object Reference)
   > Bind Event to On View Target Changed(self:Fort Player Controller Object Reference=@n13, Delegate=Create Event()) {is=title K2Node_AddDelegate}
   > OnViewTargetChanged!(
    execute
    self
    PlayerController:Fort Player Controller Object Reference
    OldViewTarget:Actor Object Reference
    NewViewTarget:Actor Object Reference=GetViewTarget!(self:Controller Object Reference=@n13) -> (ReturnValue:Actor Object Reference)
   ) -> (then) {is=title K2Node_CallFunction}
   > StopAnimation(InAnimation=Get Anim_Reload())
   > ResetReloadOnConstruct()
   > PlayAnimationReverse(InAnimation=Get Anim_HideAmmo() -> (Anim_HideAmmo:Widget Animation Object Reference), PlaybackSpeed=2)
   > HideAmmoBarsTimer()
   n14 = Get SMG_WeaponRef() -> (SMG_WeaponRef:B Cold Fern Dual Weapon Object Reference)
   > Bind Event to On Start Firing(self=@n14, Delegate=Create Event())
   > Bind Event to On Persistent Fire Stopped(self=@n14, Delegate=Create Event())
   PrintString(
    execute=@n12.CastFailed
    InString=EventConstruct: Could not cast OwningPlayer to PlayerController
    TextColor=(R=1,G=0.0701,B=0,A=1)
    Duration=5
   )
   Event Tick()
   > n15 = Sequence()
   > n16 = Get!(execute:Exec) -> (then:Exec, else:Exec, SMG_WeaponRef:B Cold Fern Dual Weapon Object Reference)
   > n17 = GetDynamicMaterial(self=Get Image_Reticle() -> (Image_Reticle:Image Object Reference))
   n18 = Get OwningPlayerControllerRef()
   > SetScalarParameterValue(
    self=@n17
    ParameterName=Spread
    Value=float * float(
     A=GetSpreadToFieldOfViewRatio!(self:Fort Weapon Ranged Object Reference=@n16, PlayerController:Fort Player Controller Object Reference=@n18, bUseMinimumSpread:bool=false) -> (ReturnValue:float)
     B=7000
    ) {is=title K2Node_PromotableOperator}
   )
   n19 = Get!(execute=@n15.then_1) -> (then, else, Shotgun_WeaponBPRef:B Dual Weapon Shotgun Wild Estate Cold Fern Object Reference)
   > n20 = GetDynamicMaterial(self=Get Image_Reticle())
   n21 = Get OwningPlayerControllerRef()
   > SetScalarParameterValue(
    self=@n20
    ParameterName=Shotgun_Spread
    Value=float * float(A=GetSpreadToFieldOfViewRatio(self=@n19, PlayerController=@n21, bUseMinimumSpread=false), B=1000)
   )
   PlayAnimationForward(execute=@n8.then, InAnimation=Get Anim_HideAmmo())
  ]}
 InterruptFadeOut {nodes [
   InterruptFadeOut() {is=title K2Node_FunctionEntry}
   n1 = Get Anim_HideAmmo()
   > n2 = Branch(Condition=IsAnimationPlaying(InAnimation=@n1))
   > n3 = StopAnimation(InAnimation=@n1)
   Set RenderOpacity!(
    execute=[@n3.then, @n2.else]
    RenderOpacity:float=1
    self:Widget Object Reference=Get AmmoCounter() -> (AmmoCounter:Overlay Object Reference)
   ) -> (then, Output_Get:float)
  ]}
 ResetReloadOnConstruct {nodes [
   ResetReloadOnConstruct() {is=title K2Node_FunctionEntry}
   > StopAnimation(InAnimation=Get Anim_Reload())
   > n1 = GetDynamicMaterial(self=Get Image_AmmoCounter() -> (Image_AmmoCounter:Image Object Reference))
   n2 = GetLocalMagazineAmmoCount!(self:Fort Weapon Object Reference=Get SMG_WeaponRef()) -> (ReturnValue:int=0)
   n3 = GetBulletsPerClip!(self:Fort Weapon Object Reference=Get SMG_WeaponRef()) -> (ReturnValue:int=0)
   > SetScalarParameterValue(self=@n1, ParameterName=AnimateProgress, Value=SafeDivide(A="To Float (Integer)"(InInt=@n2), B="To Float (Integer)"(InInt=@n3)))
   > SetScalarParameterValue(self=@n1, ParameterName=IsReload, Value=0)
   > n4 = GetDynamicMaterial(self=Get Image_Reticle())
   > SetScalarParameterValue(self=@n4, ParameterName=Master Opacity, Value=1)
   > HideAmmoBarsTimer()
  ]}
 ShotgunAmmoChanged {nodes [
   n1 = ShotgunAmmoChanged!() -> (then, LocalCount:int, LocalRemaining:int) {is=title K2Node_FunctionEntry}
   > n2 = GetDynamicMaterial(self=Get Image_AmmoPip1() -> (Image_AmmoPip1:Image Object Reference))
   > SetVectorParameterValue(self=@n2, ParameterName=MainColor, Value=SelectColor(A=(R=1,G=1,B=1,A=1), bPickA="Equal (Integer)"(A=@n1, B=2)))
   > n3 = GetDynamicMaterial(self=Get Image_AmmoPip2() -> (Image_AmmoPip2:Image Object Reference))
   > SetVectorParameterValue(self=@n3, ParameterName=MainColor, Value=SelectColor(A=(R=1,G=1,B=1,A=1), bPickA="integer > integer"(A=@n1)))
  ]}
 SmgAmmoChanged {nodes [
   n1 = SmgAmmoChanged!() -> (then, LocalCount:int, LocalRemaining:int) {is=title K2Node_FunctionEntry}
   > n2 = GetDynamicMaterial(self=Get Image_AmmoCounter())
   n3 = GetBulletsPerClip(self=Get SMG_WeaponRef()) -> (ReturnValue=0)
   > SetScalarParameterValue(self=@n2, ParameterName=AnimateProgress, Value=float / float(A="To Float (Integer)"(InInt=@n1), B="To Float (Integer)"(InInt=@n3)))
   n4 = Get SMG_WeaponRef()
   > SetScalarParameterValue(self=@n2, ParameterName=SegmentsCount, Value="To Float (Integer)"(InInt=GetBulletsPerClip(self=@n4) -> (ReturnValue=0)))
  ]}
 UnbindSMGEvents {nodes [
   UnbindSMGEvents() {is=title K2Node_FunctionEntry}
   > n1 = Get!(execute) -> (then, else, SMG_WeaponRef:B Cold Fern Dual Weapon Object Reference)
   > Unbind Event from On Local Ammo Changed(self:Fort Weapon Object Reference=@n1, Delegate=Create Event()) {is=title K2Node_RemoveDelegate}
   > Unbind Event from On Local Reload Started(self:Fort Weapon Object Reference=@n1, Delegate=Create Event()) {is=title K2Node_RemoveDelegate}
   PrintString(execute=@n1.else, InString="UnbindSMGEvents: SMG WeaponRef is invalid!", TextColor=(R=1,G=0.0701,B=0,A=1), Duration=5)
  ]}
 UnbindShotgunEvents {nodes [
   UnbindShotgunEvents() {is=title K2Node_FunctionEntry}
   > n1 = Get!(execute) -> (then, else, Shotgun_WeaponBPRef:B Dual Weapon Shotgun Wild Estate Cold Fern Object Reference)
   > Unbind Event from On Local Ammo Changed(self=@n1, Delegate=Create Event())
   > Unbind Event from On Start Firing(self:B Dual Weapon Cold Fern Generic Athena Object Reference=@n1, Delegate=Create Event()) {is=title K2Node_RemoveDelegate}
   > Unbind Event from On Persistent Fire Stopped(self:B Dual Weapon Cold Fern Generic Athena Object Reference=@n1, Delegate=Create Event()) {is=title K2Node_RemoveDelegate}
   PrintString(execute=@n1.else, InString="UnbindShotgunEvents: ShotgunWeaponRef is invalid!", TextColor=(R=1,G=0.0701,B=0,A=1), Duration=5)
  ]}
]
variables [
 AutoHideTime {type=TimerHandle}
 isReloading {type=boolean}
 OffhandWeaponComp {type=FortOffhandComponent, default=null}
 OwningPlayerControllerRef {type=FortPlayerController, default=null}
 Shotgun_WeaponBPRef {type=B_DualWeapon_Shotgun_WildEstate_ColdFern_C, default=null}
 SMG_WeaponRef {type=B_ColdFern_DualWeapon_C, default=null}
]
widgets {
 info {
  inheritedWidgetCount=1
  parentClass=/Script/FortniteUI.FortWeaponReticleExtensionWidgetBase
  rootWidgetClass=&Overlay
  widgetCount=9
 }
 widgets [
  Overlay_28 {widgetClassPath=&Overlay, widget=Overlay_28}
  Image_Reticle {
   widgetClassPath=&Image
   bIsVariable=true
   parent=Overlay_28
   widget=Image_Reticle
   properties {brush {
     bIsDynamicallyLoaded=false
     drawAs=Image
     imageType=FullColor
     mirroring=NoMirror
     resourceName=None
     resourceObject=/WildEstateWeapons/UI/ColdFernReticle/Materials/MI_UI_ColdFern_Reticle
     tiling=NoTile
     imageSize=<128 128>
     margin=<ltrb 0 0 0 0>
     outlineSettings {
      bUseBrushTransparency=false
      roundingType=HalfHeightRadius
      width=0
      color {colorUseRule=UseColor_Specified, specifiedColor=<rgba 0 0 0 0>}
      cornerRadii=<0 0 0 0>
     }
     tintColor {colorUseRule=UseColor_Specified, specifiedColor=<rgba 1 1 1 1>}
     uVRegion {bIsValid=false, max=<0 0>, min=<0 0>}
    }}
   slot {class=OverlaySlot, properties {horizontalAlignment=HAlign_Center, verticalAlignment=VAlign_Center}}
  }
  AmmoCounter {widgetClassPath=&Overlay, bIsVariable=true, parent=Overlay_28, widget=AmmoCounter, slot {class=OverlaySlot}}
  PipContainer {
   widgetClassPath=&VerticalBox
   bIsVariable=true
   parent=AmmoCounter
   widget=PipContainer
   properties {renderTransform {angle=45, scale=<1 1>, shear=<0 0>, translation=<160 154>}}
   slot {class=OverlaySlot}
  }
  Image_AmmoPip1 {
   widgetClassPath=&Image
   bIsVariable=true
   parent=PipContainer
   widget=Image_AmmoPip1
   properties {
    brush {
     bIsDynamicallyLoaded=false
     drawAs=Image
     imageType=FullColor
     mirroring=NoMirror
     resourceName=None
     resourceObject=/WildEstateWeapons/UI/ColdFernReticle/Materials/M_UI_ColdFern_Pip
     tiling=NoTile
     imageSize=<14 14>
     margin=<ltrb 0 0 0 0>
     outlineSettings {
      bUseBrushTransparency=false
      roundingType=HalfHeightRadius
      width=0
      color {colorUseRule=UseColor_Specified, specifiedColor=<rgba 0 0 0 0>}
      cornerRadii=<0 0 0 0>
     }
     tintColor {colorUseRule=UseColor_Specified, specifiedColor=<rgba 1 1 1 1>}
     uVRegion {bIsValid=false, max=<0 0>, min=<0 0>}
    }
    renderTransform {angle=-8, scale=<1 1>, shear=<0 0>, translation=<0 0>}
   }
   slot {class=VerticalBoxSlot}
  }
  Spacer_121 {widgetClassPath=&Spacer, parent=PipContainer, widget=Spacer_121, properties {size=<0 0>}, slot {class=VerticalBoxSlot}}
  Image_AmmoPip2 {
   widgetClassPath=&Image
   bIsVariable=true
   parent=PipContainer
   widget=Image_AmmoPip2
   properties {
    brush {
     bIsDynamicallyLoaded=false
     drawAs=Image
     imageType=FullColor
     mirroring=NoMirror
     resourceName=None
     resourceObject=&M_UI_ColdFern_Pip
     tiling=NoTile
     imageSize=<14 14>
     margin=<ltrb 0 0 0 0>
     outlineSettings {
      bUseBrushTransparency=false
      roundingType=HalfHeightRadius
      width=0
      color {colorUseRule=UseColor_Specified, specifiedColor=<rgba 0 0 0 0>}
      cornerRadii=<0 0 0 0>
     }
     tintColor {colorUseRule=UseColor_Specified, specifiedColor=<rgba 1 1 1 1>}
     uVRegion {bIsValid=false, max=<0 0>, min=<0 0>}
    }
    renderTransform {angle=8, scale=<1 1>, shear=<0 0>, translation=<0 0>}
   }
   slot {class=VerticalBoxSlot}
  }
  Image_AmmoCounter {
   widgetClassPath=&Image
   bIsVariable=true
   parent=AmmoCounter
   widget=Image_AmmoCounter
   properties {
    visibility=HitTestInvisible
    brush {
     bIsDynamicallyLoaded=false
     drawAs=Image
     imageType=FullColor
     mirroring=NoMirror
     resourceName=None
     resourceObject=/WildEstateWeapons/UI/ColdFernReticle/Materials/MI_UI_ColdFerne_AmmoCounter
     tiling=NoTile
     imageSize=<160 160>
     margin=<ltrb 0 0 0 0>
     outlineSettings {
      bUseBrushTransparency=false
      roundingType=HalfHeightRadius
      width=0
      color {colorUseRule=UseColor_Specified, specifiedColor=<rgba 0 0 0 0>}
      cornerRadii=<0 0 0 0>
     }
     tintColor {colorUseRule=UseColor_Specified, specifiedColor=<rgba 1 1 1 1>}
     uVRegion {bIsValid=false, max=<0 0>, min=<0 0>}
    }
    renderTransform {angle=0, scale=<1 1>, shear=<0 0>, translation=<72 72>}
   }
   slot {class=OverlaySlot}
  }
  TutorialHighlight {widgetClassPath=&TutorialHighlight}
 ]
}
